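# Material function for the Octane Neon vehicle skin that sets base color and emissive on material attributes.
# An RGB mask texture layers team color, accent color and fixed tints.
# Procedural gradient, spot and V masks on UV1 shape scrolling noise into an animated neon pattern that drives both base color and emissive, with a brightness parameter.
# A legacy SharpMasks switch, reserved for Rocket League ports, can sharpen base color.
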MaterialFunction /VehicleCosmetics/Bodies/Octane/Skins/OctaneNeon/MF_Vehicle_Skin_OctaneNeon
graph {nodes [
  n1 = NamedRerouteUsage()
  n2 = NamedRerouteUsage()
  Composite!(BaseColorInput=@n1, TextureInput=@n2) {desc=Legacy - Rocket League Feature, subgraphName=SharpMasksBaseColorMultiply}
  FunctionOutput(None=SetMaterialAttributes!(MaterialAttributes, BaseColor=NamedRerouteUsage(), EmissiveColor=NamedRerouteUsage()) {attributeSetTypes [#1, #2]})
  n3 = Distance(
   A=Subtract(A=TextureCoordinate() {coordinateIndex=1, uTiling=3, vTiling=5}, B=Constant2Vector() {g=0.9, r=1.5})
   B=Constant2Vector() {g=-0.05}
  )
  NamedRerouteDeclaration(None=Saturate(None=ConstantBiasScale(None=@n3) {bias=0, scale=1})) {name=SpotMask_Out, nodeColor=<rgba 0.4531 0 1 1>}
  n4 = TextureCoordinate() {coordinateIndex=1}
  n5 = ComponentMask(None=@n4) {g=true}
  n6 = OneMinus(None=@n5)
  n7 = ConstantBiasScale(None=@n6) {bias=-0.99, scale=15}
  n8 = Saturate(None=Subtract(A=ConstantBiasScale(None=@n6) {bias=-0.85, scale=5}, B=Saturate(None=@n7)))
  n9 = Add(A=Multiply(A=@n4, B=Constant2Vector() {g=1, r=0.9}), B=Constant2Vector() {g=-0.7, r=0.05})
  n10 = ComponentMask(None=@n9) {g=true}
  n11 = ComponentMask(None=@n9) {r=true}
  n12 = Saturate(None=Abs(None=ConstantBiasScale(None=@n11) {bias=-0.5, scale=2}))
  n13 = TextureCoordinate() {coordinateIndex=1, uTiling=3, vTiling=1.75}
  n14 = Constant2Vector() {g=0.9, r=1.5}
  n15 = ConstantBiasScale(None=Distance(A=Subtract(A=@n13, B=@n14), B=Constant2Vector() {g=-0.05})) {bias=-0.5, scale=1.5}
  n16 = Add(A=@n8, B=Multiply(A=Add(A=@n10, B=@n12), B=Saturate(None=@n15)))
  n17 = Saturate(None=ConstantBiasScale(None=@n6) {bias=-0.02, scale=22})
  NamedRerouteDeclaration(None=Subtract(A=Saturate(None=@n16), B=OneMinus(None=@n17))) {name=VMask_Out, nodeColor=<rgba 1 0 0.7883 1>}
  NamedRerouteDeclaration(None=@n5) {name=TexCoordG_Out, nodeColor=<rgba 0.4694 1 0 1>}
  NamedRerouteDeclaration(None=@n8) {name=GradTop_Out, nodeColor=<rgba 0 1 0.6016 1>}
  n18 = NamedRerouteUsage()
  n19 = NamedRerouteUsage()
  n20 = TextureCoordinate() {coordinateIndex=1}
  n21 = Multiply(A=Constant2Vector() {g=1, r=3}, B=@n20)
  n22 = ConstantBiasScale(None=TextureSample(UVs=Panner(Coordinate=@n21) {speedY=1.25}) {texture=/VehicleCosmetics/SharedTextures/T_Universal_Noise_Heightmap_30}.RGB) {bias=-0.5, scale=0.04}
  n23 = NamedRerouteUsage()
  n24 = Panner(Coordinate=Multiply(A=Constant2Vector() {g=256, r=256}, B=@n20)) {speedX=1, speedY=0.1}
  n25 = Constant2Vector() {g=1.5, r=2.5}
  n26 = TextureSample(UVs=Panner(Coordinate=Multiply(A=@n25, B=@n20)) {speedY=1.25}) {texture=&T_Universal_Noise_Heightmap_30}
  n27 = ConstantBiasScale(None=Multiply(
    A=TextureSample(UVs=@n24) {texture=/VehicleCosmetics/SharedTextures/T_Universal_Noise_Heightmap_34}.RGB
    B=ConstantBiasScale(None=@n26.RGB) {bias=-0.6, scale=5}
   )) {bias=-0.02, scale=3}
  n28 = ConstantBiasScale(None=LinearInterpolate(A=Multiply(A=@n19, B=@n22), B=Multiply(A=@n23, B=@n27), Alpha=Saturate(None=@n27))) {bias=-0.01, scale=20}
  n29 = Panner(Coordinate=@n20) {speedY=1}
  n30 = Saturate(None=ConstantBiasScale(None=TextureSample(UVs=@n29) {texture=&T_Universal_Noise_Heightmap_30}.RGB) {bias=-0.2, scale=5})
  n31 = Panner(Coordinate=Multiply(A=Constant2Vector() {g=6, r=1.5}, B=@n20)) {speedY=4}
  n32 = Saturate(None=ConstantBiasScale(None=TextureSample(UVs=@n31) {texture=&T_Universal_Noise_Heightmap_30}.RGB) {bias=-0.5, scale=8})
  n33 = Saturate(None=ConstantBiasScale(None=NamedRerouteUsage()) {bias=-0.85, scale=7})
  n34 = Saturate(None=ConstantBiasScale(None=NamedRerouteUsage()) {bias=-0.4, scale=1})
  n35 = NamedRerouteUsage()
  n36 = Add(
   A=LinearInterpolate(A=Saturate(None=@n28), B=Add(A=@n30, B=@n32), Alpha=Add(A=@n33, B=@n34))
   B=Saturate(None=ConstantBiasScale(None=@n35) {bias=-0.85, scale=4})
  )
  n37 = OneMinus(None=NamedRerouteUsage())
  n38 = Constant() {r=7.5}
  n39 = NamedRerouteUsage()
  n40 = Subtract(A=Saturate(None=Multiply(A=@n37, B=@n38)), B=Saturate(None=ConstantBiasScale(None=@n39) {bias=-0.3, scale=1}))
  n41 = Saturate(None=Add(A=Add(A=@n36, B=@n34), B=OneMinus(None=@n40)))
  NamedRerouteDeclaration(None=LinearInterpolate(A=NamedRerouteUsage(), B=@n18, Alpha=@n41) {desc=BC}) {name=NeonBC_Out, nodeColor=<rgba 0 1 0.625 1>}
  NamedRerouteDeclaration(None=VectorParameter() {group=Color, parameterName=BaseColor, sortPriority=1, defaultValue=<rgba 0 0.2 1 1>}.RGB) {name=TeamColor, nodeColor=<rgba 0.9219 0 1 1>}
  NamedRerouteDeclaration(None=VectorParameter() {group=Color, parameterName=AccentColor, sortPriority=2, defaultValue=<rgba 1 1 1 1>}.RGB) {name=AccentColor, nodeColor=<rgba 0.9219 0 1 1>}
  MaterialExpressionReroute_110 = Reroute(None=@n1)
  n42 = Constant() {r=-0.5}
  MaterialExpressionReroute_111 = Reroute(None=@n2)
  n43 = Power(Base=@MaterialExpressionReroute_111.TextureInput) {constExponent=0.4545}
  n44 = ComponentMask(None=Abs(None=Add(A=@n42, B=@n43))) {g=true}
  n45 = Add(A=@n44, B=@n44)
  n46 = Power(Base=@n44) {constExponent=8}
  n47 = Clamp(None=Power(Base=Add(A=@n45, B=@n46)) {constExponent=0.25}, Min=Constant() {r=0.5}) {clampMode=CMODE_ClampMin, minDefault=0.5}
  n48 = StaticSwitchParameter(True=Multiply(A=@MaterialExpressionReroute_110.BaseColorInput, B=Saturate(None=@n47)), False=@MaterialExpressionReroute_110.BaseColorInput) {
   desc="DO NOT USE FOR NEW ASSETS! This is intended for porting content from Rocket League only."
   group=Legacy_PORTINGONLY
   parameterName=SharpMasks
  }
  MaterialExpressionReroute_112 = Reroute(None=@n48)
  NamedRerouteDeclaration(None=@MaterialExpressionReroute_112.Result) {name=BaseColor, nodeColor=<rgba 0 1 0.7656 1>}
  NamedRerouteDeclaration(None=LinearInterpolate(B=@n18, Alpha=@n41) {desc=Emissive}) {name=NeonEmis_Out, nodeColor=<rgba 1 0.3047 0 1>}
  n49 = NamedRerouteUsage()
  n50 = TextureSample() {samplerType=SAMPLERTYPE_Masks, texture=&/T_Octane_Neon_RGB}
  n51 = LinearInterpolate(A=@n49, B=NamedRerouteUsage(), Alpha=@n50.R)
  n52 = Saturate(None=Add(A=Constant3Vector() {constant=<rgba 0.3446 0.031 0.123 1>}, B=ConstantBiasScale(None=@n49) {bias=0.2, scale=1}))
  NamedRerouteDeclaration(None=LinearInterpolate(A=LinearInterpolate(A=@n51, B=@n52, Alpha=@n50.B), B=Constant3Vector() {constant=<rgba 0.0101 0.0108 0.0112 0>}, Alpha=@n50.G)) {name=ColorLayer_Out, nodeColor=<rgba 0 0.0156 1 1>}
  NamedRerouteDeclaration(None=Multiply(A=NamedRerouteUsage(), B=ScalarParameter() {defaultValue=2, group=Color, parameterName=EmissiveBrightness, sortPriority=7})) {name=EmissiveOutput, nodeColor=<rgba 0 0.9297 1 1>}
  PinBase() {reroutePins [
    BaseColorInput {expression=&self:MaterialExpressionReroute_110}
    TextureInput {expression=&self:MaterialExpressionReroute_111}
   ]}
  PinBase!(Result=@n48) {reroutePins [Result {expression=&self:MaterialExpressionReroute_112}]}
 ]}
