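# Clear-coat material for the OctaneNeon skin on the Octane vehicle body, usable on skeletal and deformed meshes.
# The skin function's base color overrides both Base Color and Emissive on the shared body base, so the skin pattern glows.
# The result feeds damage, then de-rez transition functions; a bottom-normal reroute is declared for clear coat.
# An Apply_Imperfections static switch branches off the same attributes, with no consumer shown.
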Material /VehicleCosmetics/Bodies/Octane/Skins/OctaneNeon/MAT_Vehicle_Skin_OctaneNeon
graph {
 properties {
  bUseMaterialAttributes=true
  bUsedWithMeshDeformer=true
  bUsedWithSkeletalMesh=true
  pixelDepthOffsetMode=PDOM_Legacy
  shadingModel=MSM_ClearCoat
  usedShadingModels=Clear Coat
 }
 nodes [
  ClearCoatNormalCustomOutput!(None=NamedRerouteUsage())
  n1 = GetMaterialAttributes(None=MaterialFunctionCall() {materialFunction=&/MF_Vehicle_Skin_OctaneNeon}.Result) {attributeGetTypes [#1, #2]}
  n2 = GetMaterialAttributes(None=SetMaterialAttributes!(
    MaterialAttributes=MaterialFunctionCall() {materialFunction=/VehicleCosmetics/Materials/MaterialFunctions/MF_Vehicle_Body_Base}.Result
    Base Color=@n1.Base Color
    Emissive Color=@n1.Base Color
   ) {attributeSetTypes [#1, #2]}) {attributeGetTypes [#3]}
  MaterialFunctionCall(In=MaterialFunctionCall(Material=@n2.MaterialAttributes) {materialFunction=&MF_BaseVehicle_Damage}.Result) {materialFunction=&MF_VehicleDeRez_Transition}
  NamedRerouteDeclaration(None=@n2.ClearCoatBottomNormal) {name=BodyBase_BottomNormal, nodeColor=<rgba 0 1 0.3672 1>}
  StaticSwitchParameter(
   True=MaterialFunctionCall!(MaterialAttributes=@n2.MaterialAttributes) {materialFunction=/VehicleCosmetics/Materials/ExtraFunctions/MF_Imperfection}.Imperfection_MA
   False=@n2.MaterialAttributes
  ) {group=Imperfection, parameterName=Apply_Imperfections, sortPriority=-1}
 ]
}
