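# Material function for Octane Neon vehicle skin attachments, producing a glowing two-color gradient and an animated vertex wobble.
# A mask built from inverted V texture coordinates blends AccentColor and BaseColor, feeding base color and, scaled by EmissiveValue, emissive.
# Object-relative position in local space plus time drives two sine waves that form World Position Offset, applied without transforming back from local space.
# The results are packed into a material attributes output.
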MaterialFunction /VehicleCosmetics/Bodies/Octane/Skins/OctaneNeon/MF_Vehicle_Skin_OctaneNeon_Attachments
graph {nodes [
  n1 = OneMinus(None=ComponentMask(None=TextureCoordinate()) {g=true})
  n2 = ConstantBiasScale(None=@n1) {bias=-0.99, scale=15}
  n3 = Saturate(None=Subtract(A=ConstantBiasScale(None=@n1) {bias=-0.85, scale=3}, B=Saturate(None=@n2)))
  n4 = LinearInterpolate(
   A=VectorParameter() {group=Color, parameterName=AccentColor, sortPriority=2, defaultValue=<rgba 1 1 1 1>}.RGB
   B=VectorParameter() {group=Color, parameterName=BaseColor, sortPriority=1, defaultValue=<rgba 0 0.2 1 1>}.RGB
   Alpha=Saturate(None=ConstantBiasScale(None=@n3) {bias=-0.1, scale=0.8})
  )
  n5 = ScalarParameter() {defaultValue=2, group=Color, parameterName=EmissiveValue}
  n6 = Subtract(A=WorldPosition().XYZ, B=ObjectPositionWS())
  n7 = Add(
   A=Multiply(A=Transform(None=@n6) {transformSourceType=TRANSFORMSOURCE_World, transformType=TRANSFORM_Local}) {constB=2}
   B=Multiply(A=Time()) {constB=5}
  )
  n8 = Sine(None=ComponentMask(None=@n7) {r=true}) {period=6}
  n9 = Multiply(A=Constant3Vector() {constant=<rgba 0.5 0.5 0.1 0>}, B=Multiply(B=@n8) {constA=0.5})
  n10 = Sine(None=ComponentMask(None=@n7) {g=true}) {period=1.5}
  n11 = Saturate(None=Multiply(A=Multiply(A=@n8, B=@n10), B=Constant3Vector() {constant=<rgba 8 0 0 0>}))
  FunctionOutput(None=SetMaterialAttributes!(MaterialAttributes, BaseColor=@n4, EmissiveColor=Multiply(A=@n4, B=@n5), WorldPositionOffset=Add(A=@n9, B=@n11)) {attributeSetTypes [#1, #2, #3]})
 ]}
